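# Post-process animation Blueprint for the TroopFactor male head that adjusts tongue rotation as the jaw opens.
# Each update reads the jaw_open_pose curve and multiplies two clamped range maps into tongueRotation, ramping in at slight opening and fading out at high values.
# That value drives the interpolated alpha of a Modify Bone node, which adds a fixed bone-space roll to the tongue on the input pose.
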AnimBlueprint /BRCosmetics/Characters/Player/Male/Medium/Heads/M_MED_TroopFactor_Head/Meshes/M_MED_TroopFactor_Head_PostProcessAnimBP
target_skeleton=&/Fortnite_Base_Head_Export_TroopFactor_Skeleton
graphs [
 AnimGraph {nodes [
   n1 = "Transform (Modify) Bone - Bone: tongue"(
    ComponentPose=Local To Component(LocalPose=Input Pose())
    Alpha=Get tongueRotation() -> (tongueRotation:double)
    Translation="0,0,0"
    Rotation="0,0,-12"
    Scale="1,1,1"
   ) {
    rotationMode=BMM_Additive
    rotationSpace=BCS_BoneSpace
    translationSpace=BCS_WorldSpace
    alphaScaleBiasClamp {bInterpResult=true}
    boneToModify {boneName=tongue}
    rotation {roll=-12}
   }
   Output Pose(Result=Component To Local(ComponentPose=@n1))
  ]}
 EventGraph {nodes [
   Event BlueprintUpdateAnimation()
   n1 = GetCurveValue(CurveName=jaw_open_pose)
   > Set tongueRotation(tongueRotation:double=float * float(A=MapRangeClamped(Value=@n1, InRangeA=0, InRangeB=0.1), B=MapRangeClamped(Value=@n1, InRangeA=1.4, InRangeB=2, OutRangeA=1, OutRangeB=0)) {is=title K2Node_PromotableOperator}) -> (Output_Get:double)
   TryGetPawnOwner()
  ]}
]
variables [tongueRotation {type=number}]
